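# Niagara system for the Elm vehicle boost cosmetic, a blue flickering exhaust cone with lightning-like ribbon trails.
# NE_VC_Cone spawns short-lived sprite/mesh particles from three probabilistic spawn groups, using the group as VisibilityTag and to pick offsets.
# RibbonN001 drives four ribbon renderers with sine/cosine-pulsed colors and widths, offset positions and fading inherited movement.
# User.GameModeID selects cone offsets and ribbon velocity; painted-variant overrides are disabled.
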NiagaraSystem /VehicleCosmetics/Boosts/Boost_Elm/FX/NS_VFX_Elm
dependencies {
 usedDataInterfaces [
  &NiagaraDataInterfaceArrayColor
  &NiagaraDataInterfaceArrayFloat3
  &NiagaraDataInterfaceCurve
  &NiagaraDataInterfaceSkeletalMesh
 ]
 usedDynamicInputs []
 usedModules [
  &self:SystemSpawnScript.SystemScriptSource.SystemScriptGraph.NiagaraNodeAssignment_0.SetVariables_#1
  &SystemState
 ]
}
emitters {
 NE_VC_Cone {
  emitterUpdateScript [
   &EmitterState()
   &SpawnRate(Spawn Probability=0.3, SpawnRate=20, Use Spawn Probability=true)
   &SpawnRate(Spawn Probability=0.2, SpawnGroup=1, SpawnRate=30, Use Spawn Probability=true) {module=SpawnRate001}
   &SpawnRate(Spawn Probability=0.3, SpawnGroup=2, SpawnRate=40, Use Spawn Probability=true) {module=SpawnRate002}
  ]
  particleSpawnScript [
   &VC_PaintedVariantsOverride() {enabled=false}
   &InitializeParticle(
    Color=<rgba 0.0497 0.0814 1 1>
    Lifetime Max=0.08
    Lifetime Min=0.05
    Lifetime Mode=::NewEnumerator1
    Mesh Scale Mode=::NewEnumerator1
    Mesh Uniform Scale Max=3.6
    Mesh Uniform Scale Min=2.7
    Position Offset=&AddVectorToPosition(
     Position=&MergeHelper_VectorSwitch(
      0=&SelectVectorFromArray(Direct Array Index=$User.GameModeID, Vector Selection Array=DataInterface() {propertyValues {FloatData [<-10 0 0>, <-30 1 1>, <-40 0 0>]}})
      1=&SelectVectorFromArray(Direct Array Index=$User.GameModeID, Vector Selection Array=DataInterface() {propertyValues {FloatData [<-10 0 0>, <-20 1 1>, <-30 0 0>]}})
      2=&SelectVectorFromArray(Direct Array Index=$User.GameModeID, Vector Selection Array=DataInterface() {propertyValues {FloatData [<-10 0 0>, <-5 1 1>, <-5 0 0>]}})
      3=?
      Index=&SpawnGroup()
      Num=::NewEnumerator1
     )
     Vector=&Add_Vector(
      A=&Add_Vector(A=$User.SingleBoostPos, B=$User.DTCarOffset)
      B=&RandomVector(Vector Scale=0.4)
     )
    )
    UsePositionOffset=true
    Write Material Random=true
   )
   &AddVelocity(Velocity=<-1500 0 0>) {enabled=false}
   &InitialMeshOrientation(
    Mesh Orientation Mode=::NewEnumerator7
    Rotation=&RandomRangeVector(Minimum=<-1 -0.12 -0.11>, Maximum=<1 0.12 0.12>)
    Use Rotation Vector=true
   )
   /SetVariables(Particles.VisibilityTag=$Emitter.SpawnGroup) {isSetParameters=true}
  ]
  particleUpdateScript [
   &ParticleState()
   &ScaleMeshSize(Scale Factor=&VectorFromFloat(Value=&FloatFromCurve(FloatCurve=DataInterface() {propertyValues {Curve {
         defaultValue=1
         keys [
          {arriveTangent=0.7813, interpMode=RCIM_Cubic, leaveTangent=0.7813, tangentMode=RCTM_User, time=0, value=0.8}
          {interpMode=RCIM_Cubic, leaveTangent=0.0757, time=0.6, value=1.1}
         ]
        }}})))
   &SolveForcesAndVelocity(Clamp Velocity=false, Limit Acceleration=false) {enabled=false}
   &ScaleColor(
    Scale RGB=&VectorFromFloat(Value=&Multiply_Float(
      A=&FloatFromCurve(FloatCurve=DataInterface() {propertyValues {Curve {keys [
           {arriveTangent=9.4197, interpMode=RCIM_Cubic, leaveTangent=9.4197, tangentMode=RCTM_User, time=0, value=0.6}
           <at 0.15 4 RCIM_Cubic>
           {arriveTangentWeight=0.6608, interpMode=RCIM_Cubic, tangentMode=RCTM_User, time=1, value=0.6}
          ]}}})
      B=150
     ))
    ScaleA=true
    ScaleRGB=true
   )
  ]
  rendererClasses [&NiagaraSpriteRendererProperties, &NiagaraMeshRendererProperties]
  renderers [
   {rendererClass=&NiagaraSpriteRendererProperties}
   {rendererClass=&NiagaraMeshRendererProperties, rendererIndex=1}
   {rendererClass=&NiagaraMeshRendererProperties, rendererIndex=2}
   {rendererClass=&NiagaraMeshRendererProperties, rendererIndex=3}
  ]
 }
 RibbonN001 {
  emitterSpawnScript [/SetVariables(Emitter.MaskShifter:LinearColor=&SelectLinearColorFromArray(
     Array Sampling Mode=::NewEnumerator1
     Direct Array Index=0
     Color Selection Array=DataInterface() {propertyValues {ColorData [<rgba 1 1 1 1>, <rgba 0 1 0 1>, <rgba 0 0 1 1>]}}
     Random Seed=?
    )) {isSetParameters=true}]
  emitterUpdateScript [
   &EmitterState()
   &VC_FilteredSocketCount() {enabled=false}
   &VC_RateQualityHandler(Epic Spawn Rate=40, High Spawn Rate=40, Low Spawn Rate=30, Use Custom Rate?=true)
   &VC_SpawnRibbon_v1(SpawnRate=$Emitter.Spawn Rate)
  ]
  particleSpawnScript [
   &VC_PaintedVariantsOverride() {enabled=false}
   &VC_InitializeRibbon_v1(
    Custom Offset=true
    Lifetime=0.8
    Local Vector Offset=&AddVectorToPosition(Position=<0 0 0>, Vector=&Add_Vector(A=$User.SingleBoostPos, B=$User.DTCarOffset))
    Offset Amount=1
    Ribbon Width=120
   )
   &AddVelocity(Velocity=&SelectVectorFromArray(
     Direct Array Index=$User.GameModeID
     Vector Selection Array=DataInterface() {propertyValues {FloatData [<-2000 0 0>, <-1800 1 1>, <-1500 0 0>]}}
    )) {module=AddVelocity001}
  ]
  particleUpdateScript [
   &ParticleState()
   &InheritSourceMovement(
    Applied Position Scale=&Multiply_VectorByFloat(
     Vector=<1 1 1>
     Float=&FloatFromCurve(FloatCurve=DataInterface() {propertyValues {Curve {keys [{interpMode=RCIM_Cubic, tangentMode=RCTM_Break, time=0.3, value=1}, {interpMode=RCIM_Cubic, tangentMode=RCTM_Break, time=1, value=0}]}}})
    )
    ApplyForce=false
    ApplyPosition=true
    ApplyPositionFalloff=false
   )
   &SolveForcesAndVelocity(Clamp Velocity=false, Limit Acceleration=false)
   &VC_UpdateRibbon_v1()
   /SetVariables(
    Particles.Index1:Vector4f=&MakeVector4(X=1, Y=13, Z=&RandomRangeFloat(Minimum=50, Maximum=1), W=0)
    Particles.Index2:Vector4f=&MakeVector4(X=1, Y=7, Z=&RandomRangeFloat(Minimum=50, Maximum=1), W=0)
   ) {isSetParameters=true}
   /SetVariables(
    Particles.lightningcolor1:LinearColor=&Multiply_LinearColorByFloat(
     Linear Color=&MakeLinearColorFromVectorAndFloat(
      "Vector (RGB)"=&MakeVectorFromLinearColorRGB(Color=<rgba 0.04 0.056 1 1>)
      "Float (Alpha)"=&FloatFromCurve(FloatCurve=DataInterface() {propertyValues {Curve {keys [<at 0 0>,<at 0.1 0.5>,<at 0.7 0.5>,<at 0.9 0>]}}})
     )
     Float=&Cosine(Normalized Angle=$Emitter.Age, Period=0.4, Scale=15, Bias=11)
    )
    Particles.lightningcolor2:LinearColor=&Multiply_LinearColorByFloat(
     Linear Color=&MakeLinearColorFromVectorAndFloat(
      "Vector (RGB)"=&MakeVectorFromLinearColorRGB(Color=<rgba 0.04 0.056 1 1>)
      "Float (Alpha)"=&FloatFromCurve(FloatCurve=DataInterface() {propertyValues {Curve {keys [<at 0 0>,<at 0.1 0.5>,<at 0.7 0.5>,<at 0.9 0>]}}})
     )
     Float=&Sine(Normalized Angle=$Emitter.Age, Period=0.6, Scale=20, Bias=10)
    )
    Particles.lightningcolor3:LinearColor=&Multiply_LinearColorByFloat(
     Linear Color=&MakeLinearColorFromVectorAndFloat(
      "Vector (RGB)"=&MakeVectorFromLinearColorRGB(Color=<rgba 0.04 0.056 1 1>)
      "Float (Alpha)"=&FloatFromCurve(FloatCurve=DataInterface() {propertyValues {Curve {keys [<at 0 0>,<at 0.1 0.8>,<at 0.7 0.8>,<at 0.9 0>]}}})
     )
     Float=&Cosine(Normalized Angle=$Emitter.Age, Period=0.7, Scale=25, Bias=7)
    )
    Particles.lightningcolor4:LinearColor=&Multiply_LinearColorByFloat(
     Linear Color=&MakeLinearColorFromVectorAndFloat(
      "Vector (RGB)"=&MakeVectorFromLinearColorRGB(Color=<rgba 2.8 3.9199 70 1>)
      "Float (Alpha)"=&FloatFromCurve(FloatCurve=DataInterface() {propertyValues {Curve {keys [<at 0 0>,<at 0.1 0.5>,<at 0.7 0.5>,<at 0.9 0>]}}})
     )
     Float=&Cosine(Normalized Angle=$Emitter.Age, Period=0.4, Scale=15, Bias=11)
    )
   ) {isSetParameters=true}
   /SetVariables(
    Particles.lightningwidth1:NiagaraFloat=&Multiply_Float(
     A=&FloatFromCurve(FloatCurve=DataInterface() {propertyValues {Curve {
         defaultValue=1
         keys [
          {arriveTangent=0.5666, interpMode=RCIM_Cubic, leaveTangent=0.5666, tangentMode=RCTM_User, time=0, value=0.7}
          {arriveTangent=0.3739, interpMode=RCIM_Cubic, leaveTangent=0.3739, time=0.1753, value=0.9894}
          {arriveTangent=-0.7147, interpMode=RCIM_Cubic, leaveTangent=-0.7147, time=0.7562, value=0.9828}
          {arriveTangent=-0.1422, interpMode=RCIM_Cubic, leaveTangent=-0.1422, tangentMode=RCTM_User, time=1, value=0.4}
         ]
        }}})
     B=&Cosine(Normalized Angle=$Emitter.Age, Period=0.7, Scale=60, Bias=30)
    )
    Particles.lightningwidth2:NiagaraFloat=&Multiply_Float(
     A=&FloatFromCurve(FloatCurve=DataInterface() {propertyValues {Curve {
         defaultValue=1
         keys [
          {arriveTangent=1.3608, interpMode=RCIM_Cubic, leaveTangent=1.3608, tangentMode=RCTM_User, time=0, value=0.7}
          {arriveTangent=0.3739, interpMode=RCIM_Cubic, leaveTangent=0.3739, time=0.1753, value=0.9894}
          {arriveTangent=-0.7147, interpMode=RCIM_Cubic, leaveTangent=-0.7147, time=0.7562, value=0.9828}
          <at 1 0.4 RCIM_Cubic>
         ]
        }}})
     B=50
    )
    Particles.lightningwidth3:NiagaraFloat=&Multiply_Float(
     A=&FloatFromCurve(FloatCurve=DataInterface() {propertyValues {Curve {
         defaultValue=1
         keys [
          {arriveTangent=1.407, interpMode=RCIM_Cubic, leaveTangent=1.407, tangentMode=RCTM_User, time=0, value=0.7}
          {arriveTangent=0.3739, interpMode=RCIM_Cubic, leaveTangent=0.3739, time=0.1753, value=0.9894}
          {arriveTangent=-0.7147, interpMode=RCIM_Cubic, leaveTangent=-0.7147, time=0.7562, value=0.9828}
          <at 1 0.4 RCIM_Cubic>
         ]
        }}})
     B=60
    )
    Particles.lightningwidth4:NiagaraFloat=&Multiply_Float(
     A=&FloatFromCurve(FloatCurve=DataInterface() {propertyValues {Curve {
         defaultValue=1
         keys [
          {arriveTangent=0.5666, interpMode=RCIM_Cubic, leaveTangent=0.5666, tangentMode=RCTM_User, time=0, value=0.7}
          {arriveTangent=0.3739, interpMode=RCIM_Cubic, leaveTangent=0.3739, time=0.1753, value=0.9894}
          {arriveTangent=-0.7147, interpMode=RCIM_Cubic, leaveTangent=-0.7147, time=0.7562, value=0.9828}
          {arriveTangent=-0.1422, interpMode=RCIM_Cubic, leaveTangent=-0.1422, tangentMode=RCTM_User, time=1, value=0.4}
         ]
        }}})
     B=&Cosine(Normalized Angle=$Emitter.Age, Period=0.5, Scale=35, Bias=15)
    )
   ) {isSetParameters=true}
   /SetVariables(
    Particles.lightningposition2:NiagaraPosition=&AddVectorToPosition(Position=$Particles.Position, Vector=<-5 -20 -8>)
    Particles.lightningposition3:NiagaraPosition=&AddVectorToPosition(Position=$Particles.Position, Vector=<5 20 8>)
   ) {isSetParameters=true}
  ]
  rendererClasses [&NiagaraRibbonRendererProperties]
  renderers [
   {rendererClass=&NiagaraRibbonRendererProperties}
   {rendererClass=&NiagaraRibbonRendererProperties, rendererIndex=1}
   {rendererClass=&NiagaraRibbonRendererProperties, rendererIndex=2}
   {rendererClass=&NiagaraRibbonRendererProperties, rendererIndex=3}
  ]
 }
}
summary {
 systemName=NS_VFX_Elm
 emitters [
  {emitterName=RibbonN001, rendererClasses [&NiagaraRibbonRendererProperties]}
  {emitterName=NE_VC_Cone, rendererClasses [&NiagaraSpriteRendererProperties, &NiagaraMeshRendererProperties]}
 ]
 userVariables [
  User.bIsFrontendPreview {type=NiagaraBool, defaultValue=true}
  User.bIsPainted {type=NiagaraBool}
  User.ColorVariantID {type=NiagaraInt32}
  User.DTCarOffset {type=Vector3f, description="", defaultValue {struct=Vector3f, value=<0 0 0>}}
  User.GameModeID {type=NiagaraInt32}
  User.PaintedColor {type=LinearColor, description="", defaultValue {struct=LinearColor, value=<rgba 1 1 1 1>}}
  User.PaintedScalar {type=NiagaraFloat, defaultValue=1, description=""}
  User.PortCount {type=NiagaraInt32, defaultValue=1}
  User.SingleBoostPos {type=Vector3f, description="", defaultValue {struct=Vector3f, value=<0 0 0>}}
 ]
}
